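# Parent material for the Ranked Scythe pickaxe cosmetic, layering ranked effects onto the shared 2023 character material pipeline.
# The character PreFX function feeds a ranked thin-film function, then an emissive Fresnel function borrowed from the Ranked Scythe body cosmetic, driven by the Tech Art Mask.
# A ranked_on scalar, on by default, blends between plain PreFX and ranked attributes before the shared PostFX function.
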Material /BRCosmetics/Weapons/FORT_Melee/Pickaxe_RankedScythe/Materials/M_Pickaxe_RankedScythe_Parent
graph {
 properties {
  bUseMaterialAttributes=true
  bUsedWithMeshDeformer=true
  bUsedWithNanite=true
  bUsedWithSkeletalMesh=true
  pixelDepthOffsetMode=PDOM_Legacy
  shadingModel=MSM_FromMaterialExpression
  usedShadingModels=Default Lit | Subsurface | Clear Coat | Subsurface Profile | Cloth
 }
 nodes [
  n1 = MaterialFunctionCall() {materialFunction=&MF_CharacterParent_2023_PreFX}
  n2 = MaterialFunctionCall!(Material Attributes=@n1.Result, Alpha=@n1.Result) {materialFunction=/Game/Global/Materials/Cosmetics/Ranked/MF_RankedThinFilm}
  MaterialFunctionCall(
   Material=BlendMaterialAttributes(
    A=@n1.Result
    B=MaterialFunctionCall!(Material=@n2.Result, Mask=@n1.Tech Art Mask, Fresnel=@n1.Tech Art Mask) {materialFunction=/BRCosmetics/Characters/Player/Female/Medium/Bodies/F_MED_RankedScythe/Materials/MF_RankedScythe_EmissiveFresnel}.Result
    Alpha=ScalarParameter() {defaultValue=1, parameterName=ranked_on}
   )
   M=@n1.M
   Pre Skinned Local Position=@n1.M
   Distance Blend=@n1.M
   Fresnel=@n1.M
  ) {materialFunction=&MF_CharacterParent_2023_PostFX}
 ]
}
