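# Physics asset for the RankedScythe two-handed pickaxe weapon mesh in BRCosmetics.
# It holds a single capsule body on the Melee_TwoHanded_Handle bone, offset and rotated to wrap the weapon.
# With no constraints, the weapon collides as one rigid shape rather than an articulated chain.
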PhysicsAsset /BRCosmetics/Weapons/FORT_Melee/Pickaxe_RankedScythe/Meshes/RankedScythe_Axe_PhysicsAsset
bodies {Melee_TwoHanded_Handle [{shapeType=Capsule, length=77.8248, radius=42.1467, center=<0.7662 4.437 36.8939>, rotation=<rot 1.7477 -90 -0.0082>}]}
constraints []
